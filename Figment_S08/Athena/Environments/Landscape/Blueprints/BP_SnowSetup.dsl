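; Blueprint that controls seasonal landscape snow and snow-styled minimap location fonts in Battle Royale.
; At match ready, the authority sets Snow_Amount from calendar events: full during "SP", decreasing through "SP1", otherwise zero.
; Remote clients' OnRep waits briefly, writes the curve-mapped amount to the landscape material parameter collection, then applies the snow font to POIs by rising thresholds.
; Construction script and editor events also write the snow parameter.
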(event Custom|SetFullSnowEd
  ;[1 1 1] Editor Commands
  (Rendering|Material|SetScalarParameterValue "/Game/Athena/Environments/Landscape/MPC/MPC_Landscape.MPC_Landscape" "Holiday_SnowAmount" 1e+01 :!class /Script/Engine.KismetMaterialLibrary))

(event Custom|SetSnowEd
  ;[1 1 1] Editor Commands
  (Rendering|Material|SetScalarParameterValue "/Game/Athena/Environments/Landscape/MPC/MPC_Landscape.MPC_Landscape" "Holiday_SnowAmount" (Math|Curves|GetFloatValue (Variables|Default|GetCurve) (Variables|Default|GetSnow_Amount)) :!class /Script/Engine.KismetMaterialLibrary))

(event Custom|SetMPCValue
  (Utilities|FlowControl|Delay 1.0)
  (Rendering|Material|SetScalarParameterValue "/Game/Athena/Environments/Landscape/MPC/MPC_Landscape.MPC_Landscape" "Holiday_SnowAmount" (Math|Curves|GetFloatValue (Variables|Default|GetCurve) (Variables|Default|GetSnow_Amount)) :!class /Script/Engine.KismetMaterialLibrary)
  (CallFunction|RefreshMapLocations))

(event EventBeginPlay
  (Class|FortAsyncActionAuthorityMatchReady|MatchReadyAsync
    (:then
      (CallFunction|RefreshMapLocations))
    (:OnReady
      (CallFunction|RefreshCalender(AuthorityOnly)))))

(event Custom|RefreshCalender%20%28Authority%20Only%29
  (Networking|SwitchHasAuthority
    (:Authority
      (Utilities|FlowControl|Sequence
        (:then_0
          (bind _biseventactive (CalendarEvents|GetCalendarEventInformation "SP"))
          (if _biseventactive
            (Variables|Default|SetSnow_Amount 1.0)
            (else
              (bind (_biseventactive_2 _timesincebegin _timeuntilend _timespanratio) (CalendarEvents|GetCalendarEventInformation "SP1"))
              (if _biseventactive_2
                (Variables|Default|SetSnow_Amount (- 1.0 _timespanratio))
                (else
                  (Variables|Default|SetSnow_Amount 0.0))))))))))

(event Custom|SetSnow
  (params (NewParam :type Float%20%28double-precision%29))
  (Variables|Default|SetSnow_Amount NewParam))

(event Custom|RefreshMapLocations
  (if (not (Variables|Default|GetPendingMinimapRefresh))
    (Variables|Default|SetPendingMinimapRefresh true)
    (bind _gamestate (Class|FortAsyncActionGameStateReady|GameStateReadyAsync)
      (:OnReady
        (bind _asfort_game_state_athena (Utilities|Casting|CastToFortGameStateAthena _gamestate))
        (Variables|Default|SetBRGameState _asfort_game_state_athena)
        (Variables|Default|SetPendingMinimapRefresh false)
        ;[1 1 1] Update Minimap Text
        (Utilities|FlowControl|Sequence
          (:then_0
            (CallFunction|MakeLocationVisibleinSnow (GameplayTags|MakeLiteralGameplayTag "(TagName=\"Athena.Location.POI.PolarPeak\")"))
            (CallFunction|MakeLocationVisibleinSnow (GameplayTags|MakeLiteralGameplayTag "(TagName=\"Athena.Location.POI.FrostyFlights\")"))
            (CallFunction|MakeLocationVisibleinSnow (GameplayTags|MakeLiteralGameplayTag "(TagName=\"Athena.Location.POI.HappyHamlet\")"))
            (CallFunction|MakeLocationVisibleinSnow (GameplayTags|MakeLiteralGameplayTag "(TagName=\"Athena.Location.POI.ShiftyShafts\")")))
          (:then_1
            (if (>= (Variables|Default|GetSnow_Amount) 0.1)
              (CallFunction|MakeLocationVisibleinSnow (GameplayTags|MakeLiteralGameplayTag "(TagName=\"Athena.Location.POI.TiltedTowers\")"))
              (CallFunction|MakeLocationVisibleinSnow (GameplayTags|MakeLiteralGameplayTag "(TagName=\"Athena.Location.POI.LuckyLanding\")"))
              (CallFunction|MakeLocationVisibleinSnow (GameplayTags|MakeLiteralGameplayTag "(TagName=\"Athena.Location.POI.SnobbyShores\")"))))
          (:then_2
            (if (>= (Variables|Default|GetSnow_Amount) 0.15)
              (CallFunction|MakeLocationVisibleinSnow (GameplayTags|MakeLiteralGameplayTag "(TagName=\"Athena.Location.POI.FatalFields\")"))))
          (:then_3
            (if (>= (Variables|Default|GetSnow_Amount) 0.2)
              (CallFunction|MakeLocationVisibleinSnow (GameplayTags|MakeLiteralGameplayTag "(TagName=\"Athena.Location.POI.SaltySprings\")"))))
          (:then_4
            (if (>= (Variables|Default|GetSnow_Amount) 0.25)
              (CallFunction|MakeLocationVisibleinSnow (GameplayTags|MakeLiteralGameplayTag "(TagName=\"Athena.Location.POI.PleasantPark\")"))
              (CallFunction|MakeLocationVisibleinSnow (GameplayTags|MakeLiteralGameplayTag "(TagName=\"Athena.Location.POI.LootLake\")"))))
          (:then_5
            (if (>= (Variables|Default|GetSnow_Amount) 0.3)
              (CallFunction|MakeLocationVisibleinSnow (GameplayTags|MakeLiteralGameplayTag "(TagName=\"Athena.Location.POI.HauntedHills\")"))
              (CallFunction|MakeLocationVisibleinSnow (GameplayTags|MakeLiteralGameplayTag "(TagName=\"Athena.Location.POI.DustyDivot\")"))))
          (:then_6
            (if (>= (Variables|Default|GetSnow_Amount) 0.4)
              (CallFunction|MakeLocationVisibleinSnow (GameplayTags|MakeLiteralGameplayTag "(TagName=\"Athena.Location.POI.JunkJunction\")"))
              (CallFunction|MakeLocationVisibleinSnow (GameplayTags|MakeLiteralGameplayTag "(TagName=\"Athena.Location.POI.LazyLinks\")"))
              (CallFunction|MakeLocationVisibleinSnow (GameplayTags|MakeLiteralGameplayTag "(TagName=\"Athena.Location.POI.TomatoTemple\")"))
              (CallFunction|MakeLocationVisibleinSnow (GameplayTags|MakeLiteralGameplayTag "(TagName=\"Athena.Location.POI.RetailRow\")"))
              (CallFunction|MakeLocationVisibleinSnow (GameplayTags|MakeLiteralGameplayTag "(TagName=\"Athena.Location.POI.ParadisePalms\")"))))
          (:then_7
            (if (>= (Variables|Default|GetSnow_Amount) 0.75)
              (CallFunction|MakeLocationVisibleinSnow (GameplayTags|MakeLiteralGameplayTag "(TagName=\"Athena.Location.POI.RiskyReels\")"))
              (CallFunction|MakeLocationVisibleinSnow (GameplayTags|MakeLiteralGameplayTag "(TagName=\"Athena.Location.POI.WailingWoods\")"))
              (CallFunction|MakeLocationVisibleinSnow (GameplayTags|MakeLiteralGameplayTag "(TagName=\"Athena.Location.POI.LonelyLodge\")")))))))))

(fn MakeLocationVisibleInSnow
  (params (LocationTag :type Gameplay%20Tag%20Structure%20%28by%20ref%29))
  (UI|Minimap|SetLocationFontinState (Variables|Default|GetBRGameState) LocationTag (Variables|Default|GetSnowFont) "Normal"))

(fn OnRep_Snow_Amount
  (Networking|SwitchHasAuthority
    (:Remote
      (Variables|Default|SetDidEverRepNotify true)
      (CallFunction|SetMPCValue))))

(fn ConstructionScript
  (Rendering|Material|SetScalarParameterValue "/Game/Athena/Environments/Landscape/MPC/MPC_Landscape.MPC_Landscape" "Holiday_SnowAmount" (Math|Curves|GetFloatValue (Variables|Default|GetCurve) (Variables|Default|GetSnow_Amount)) :!class /Script/Engine.KismetMaterialLibrary))
